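# Sound cue for a quiet exterior ambient of buzzing flies, routed to the ambient tile mix class with insect prop attenuation.
# A single flies wave feeds an Enveloper that holds volume near full for about four seconds, then fades to silence by about six.
# A Modulator at the root slightly randomizes pitch per play.
# A concurrency override prevents new instances once the limit is reached.
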SoundCue /Game/Sounds/Fort_Ambient_Looping/Nature_Exterior/Insect_Flies_Loop_Env5Secs_Cue
root=Modulator_0
nodes [
 {
  id=Enveloper_0
  children [WavePlayer_0]
  properties {pitchCurve {editorCurveData {keys [{time=0}]}}, volumeCurve {editorCurveData {keys [<at 0 1>,<at 3.9963 0.995>,<at 6.3888 0>]}}}
 }
 {id=Modulator_0, children [Enveloper_0], properties {pitchMax=1.1, pitchMin=0.9}}
 {id=WavePlayer_0, properties {soundWaveAssetPtr=&/fort_flies_01}}
]
properties {
 attenuationSettings=/Game/Sounds/Attenuation/Fort_Prop_Insects_Att
 bOverrideConcurrency=true
 maxDistance=1200
 soundClassObject=/Game/Sounds/FortSoundClassesAndModes/Fort_Ambient_TileMix
 volumeMultiplier=0.25
 concurrencyOverrides {resolutionRule=PreventNew}
}
